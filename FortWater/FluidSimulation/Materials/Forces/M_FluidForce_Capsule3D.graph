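# Additive unlit material that stamps a 3D capsule force into the water fluid simulation grid, allowing negative values.
# UVs map to world positions on the WaterZ plane, where a capsule distance field along axis Z around P gives a saturated mask.
# The mask scales signed force F into red and foam, from force magnitude above MinFoamForce, into green.
# A rotated BoatForceFoam texture sample does not feed that result.
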Material /FortWater/FluidSimulation/Materials/Forces/M_FluidForce_Capsule3D
graph {
 properties {
  bAllowNegativeEmissiveColor=true
  bUseTranslucencyVertexFog=false
  blendMode=BLEND_Additive
  pixelDepthOffsetMode=PDOM_Legacy
  shadingModel=MSM_Unlit
  usedShadingModels=Unlit
 }
 nodes [
  n1 = Multiply(A=Subtract(A=TextureCoordinate() {desc=UVs}) {constB=0.5}, B=ScalarParameter() {defaultValue=1024, parameterName=FluidSize})
  n2 = ComponentMask(None=VectorParameter() {parameterName=GridLocation}.RGB) {g=true, r=true}
  n3 = VectorParameter() {parameterName=P}
  n4 = Multiply(A=VectorParameter() {parameterName=Z, defaultValue=<rgba 0 0 1 0>}.RGB, B=ScalarParameter() {defaultValue=16, parameterName=R})
  n5 = MaterialFunctionCall!(
   Coordinates=AppendVector(A=Add(A=@n1, B=@n2), B=ScalarParameter() {parameterName=WaterZ})
   P0=Add(A=@n3.RGB, B=@n4)
   P1=Subtract(A=@n3.RGB, B=@n4)
   Width=ScalarParameter() {parameterName=Capsule Width}
  ) {materialFunction=/Engine/Functions/DistanceFields/Shapes/DistanceField_Capsule}
  n6 = Saturate(None=Multiply(A=Multiply(A=@n5.Result) {constB=-1}, B=ScalarParameter() {defaultValue=5, parameterName=Edge Density}))
  n7 = ScalarParameter() {defaultValue=1, parameterName=F}
  n8 = Abs(None=ScalarParameter() {defaultValue=1, parameterName=F})
  n9 = ScalarParameter() {defaultValue=0.25, parameterName=MinFoamForce}
  n10 = Multiply(A=Max(A=Subtract(A=@n8, B=@n9)) {constB=0}, B=ScalarParameter() {defaultValue=0.5, parameterName=Foam Strength}) {constB=10}
  AppendVector(A=AppendVector(A=Multiply(A=@n6, B=@n7), B=Multiply(A=@n6, B=@n10)), B=Constant())
  ScalarParameter() {defaultValue=1, parameterName=Final Opacity}
  n11 = VectorParameter() {parameterName=X, defaultValue=<rgba 1 0 0 0>}
  n12 = MaterialFunctionCall(Float2=Subtract(A=TextureCoordinate()) {constB=0.5, desc=offset}) {materialFunction=&BreakOutFloat2Components}
  n13 = VectorParameter() {parameterName=Y, defaultValue=<rgba 0 1 0 0>}
  n14 = Add(A=Multiply(A=ComponentMask(None=@n11.RGB) {g=true, r=true}, B=@n12.G), B=Multiply(A=ComponentMask(None=@n13.RGB) {g=true, r=true}, B=@n12.R))
  TextureSample(UVs=Clamp(None=Add(A=@n14) {constB=0.5})) {texture=/FortWater/Textures/BoatForceFoam}
 ]
}
