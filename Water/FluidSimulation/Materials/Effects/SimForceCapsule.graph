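# Material function for the water fluid simulation that draws a soft capsule force between points P0 and P1.
# Default coordinates map centered UVs through FluidSize and GridLocation, with WaterZ as height.
# Clamped segment projection gives squared distance, which Edge Falloff turns into an inverted linear mask, output as mask-scaled force plus mask.
# The force lerp uses AB Force Amount's R for both ends, so force stays constant along the segment.
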MaterialFunction /FortWater/FluidSimulation/Materials/Effects/SimForceCapsule
graph {
 properties {bExposeToLibrary=true, description="Draws a line between 2 points using supplied 2D coordinates. "}
 nodes [
  n1 = TextureCoordinate() {desc=UVs}
  n2 = Add(
   A=Multiply(A=Subtract(A=@n1) {constB=0.5}, B=ScalarParameter() {defaultValue=1024, parameterName=FluidSize})
   B=ComponentMask(None=VectorParameter() {parameterName=GridLocation}.RGB) {g=true, r=true}
  )
  n3 = ScalarParameter() {parameterName=WaterZ}
  n4 = FunctionInput() {bUsePreviewValueAsDefault=true, inputName=P0, sortPriority=1, previewValue=<0 0 0.5 1>}
  n5 = Subtract(A=FunctionInput(Preview=AppendVector(A=@n2, B=@n3)) {bUsePreviewValueAsDefault=true, inputName=Coordinates}, B=@n4) {desc=PA}
  n6 = Subtract(A=FunctionInput() {bUsePreviewValueAsDefault=true, inputName=P1, sortPriority=2, previewValue=<1 1 0.5 1>}, B=@n4) {desc=BA}
  n7 = Saturate(None=Divide(A=DotProduct(A=@n5, B=@n6), B=DotProduct(A=@n6, B=@n6)) {constB=1})
  n8 = Subtract(A=@n5, B=Multiply(A=@n7, B=@n6))
  Distance(A=@n8, B=Constant())
  n9 = DotProduct(A=@n8, B=@n8)
  n10 = FunctionInput() {bUsePreviewValueAsDefault=true, inputName=Edge Falloff, inputType=FunctionInput_Scalar, sortPriority=4, previewValue=<0.5 1 0 1>}
  n11 = OneMinus(None=Saturate(None=Divide(A=@n9, B=@n10)))
  n12 = FunctionInput() {bUsePreviewValueAsDefault=true, inputName=AB Force Amount, inputType=FunctionInput_Vector2, sortPriority=10, previewValue=<1 0.5 0 1>}
  n13 = MaterialFunctionCall(Float2=@n12) {materialFunction=&BreakOutFloat2Components}
  n14 = LinearInterpolate(A=@n13.R, B=@n13.R, Alpha=@n7)
  FunctionOutput(None=AppendVector(A=Multiply(A=@n11, B=@n14), B=@n11))
 ]
}
