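# Unlit alpha-composite fluid simulation force material that, going by its name, stamps a skeletal mesh's influence into FortWater as six capsules.
# Each SimForceCapsule takes two endpoint parameters, with force amounts in their alpha.
# Summed red results become a clamped force paired with foam from magnitude above a minimum, while the maximum green result drives opacity and threshold masks.
# A panned foam-texture noise branch is also present.
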Material /FortWater/FluidSimulation/Materials/Forces/M_Fluid_Force_Skelmesh
graph {
 properties {
  bAllowNegativeEmissiveColor=true
  bUseTranslucencyVertexFog=false
  bUsedWithNanite=true
  blendMode=BLEND_AlphaComposite
  pixelDepthOffsetMode=PDOM_Legacy
  shadingModel=MSM_Unlit
  usedShadingModels=Unlit
 }
 nodes [
  n1 = VectorParameter() {parameterName=p0a, defaultValue=<rgba 0.5 0.5 0.01 0>}
  n2 = VectorParameter() {parameterName=p0b, defaultValue=<rgba 0.5 0.5 0.01 0>}
  n3 = MaterialFunctionCall!(
   Coordinates
   P0=@n1.RGB
   P1=@n2.RGB
   Edge Falloff=ScalarParameter() {defaultValue=0.016, parameterName=CapsuleWidth}
   AB Force Amount=AppendVector(A=@n1.A, B=@n2.A)
  ) {materialFunction=/FortWater/FluidSimulation/Materials/Effects/SimForceCapsule}
  n4 = VectorParameter() {parameterName=p1a, defaultValue=<rgba 0.5 0.5 0.01 0>}
  n5 = VectorParameter() {parameterName=p1b, defaultValue=<rgba 0.5 0.5 0.01 0>}
  n6 = MaterialFunctionCall(
   P0=@n4.RGB
   P1=@n5.RGB
   Edge Falloff=ScalarParameter() {defaultValue=0.016, parameterName=CapsuleWidth}
   AB Force Amount=AppendVector(A=@n4.A, B=@n5.A)
  ) {materialFunction=&SimForceCapsule}
  n7 = Max(A=@n3.Result, B=@n6.Result)
  n8 = VectorParameter() {parameterName=p2a, defaultValue=<rgba 0.5 0.5 0.01 0>}
  n9 = VectorParameter() {parameterName=p2b, defaultValue=<rgba 0.5 0.5 0.01 0>}
  n10 = MaterialFunctionCall(
   P0=@n8.RGB
   P1=@n9.RGB
   Edge Falloff=ScalarParameter() {defaultValue=0.016, parameterName=CapsuleWidth}
   AB Force Amount=AppendVector(A=@n8.A, B=@n9.A)
  ) {materialFunction=&SimForceCapsule}
  n11 = VectorParameter() {parameterName=p3a, defaultValue=<rgba 0.5 0.5 0.01 0>}
  n12 = VectorParameter() {parameterName=p3b, defaultValue=<rgba 0.5 0.5 0.01 0>}
  n13 = MaterialFunctionCall(
   P0=@n11.RGB
   P1=@n12.RGB
   Edge Falloff=ScalarParameter() {defaultValue=0.016, parameterName=CapsuleWidth}
   AB Force Amount=AppendVector(A=@n11.A, B=@n12.A)
  ) {materialFunction=&SimForceCapsule}
  n14 = Max(A=@n10.Result, B=@n13.Result)
  n15 = VectorParameter() {parameterName=p4a, defaultValue=<rgba 0.5 0.5 0.01 0>}
  n16 = VectorParameter() {parameterName=p4b, defaultValue=<rgba 0.5 0.5 0.01 0>}
  n17 = MaterialFunctionCall(
   P0=@n15.RGB
   P1=@n16.RGB
   Edge Falloff=ScalarParameter() {defaultValue=0.016, parameterName=CapsuleWidth}
   AB Force Amount=AppendVector(A=@n15.A, B=@n16.A)
  ) {materialFunction=&SimForceCapsule}
  n18 = VectorParameter() {parameterName=p5a, defaultValue=<rgba 0.5 0.5 0.01 0>}
  n19 = VectorParameter() {parameterName=p5b, defaultValue=<rgba 0.5 0.5 0.01 0>}
  n20 = MaterialFunctionCall(
   P0=@n18.RGB
   P1=@n19.RGB
   Edge Falloff=ScalarParameter() {defaultValue=0.016, parameterName=CapsuleWidth}
   AB Force Amount=AppendVector(A=@n18.A, B=@n19.A)
  ) {materialFunction=&SimForceCapsule}
  n21 = ComponentMask(None=Max(A=Max(A=@n7, B=@n14), B=Max(A=@n17.Result, B=@n20.Result))) {g=true}
  Add(A=ScalarParameter() {defaultValue=0.0265, parameterName=DebugOpacity}, B=Saturate(None=@n21))
  n22 = Add(A=@n3.Result, B=@n6.Result)
  n23 = Add(A=@n10.Result, B=@n13.Result)
  n24 = ComponentMask(None=Add(A=Add(A=@n22, B=@n23), B=Add(A=@n17.Result, B=@n20.Result))) {r=true}
  n25 = ScalarParameter() {defaultValue=16, parameterName=FluidClamp}
  n26 = Subtract(A=Abs(None=@n24), B=ScalarParameter() {defaultValue=2, parameterName=Foam Min})
  AppendVector(
   A=Clamp(None=@n24, Min=Multiply(B=@n25) {constA=-1}, Max=@n25) {maxDefault=2, minDefault=-2}
   B=Multiply(A=Max(A=@n26) {constB=0}, B=ScalarParameter() {defaultValue=0.1, parameterName=FoamBoost})
  )
  Ceil(None=Subtract(A=@n21, B=ScalarParameter() {defaultValue=0.5, parameterName=Mask Threshold}) {constB=0.0005})
  Constant()
  MaterialFunctionCall(Vector 2=@n21) {materialFunction=/Engine/Functions/Engine_MaterialFunctions02/Utility/VectorLength}
  n27 = Panner(
   Coordinate=Multiply(A=TextureCoordinate(), B=ScalarParameter() {defaultValue=0.5, parameterName=NoiseTiling})
   Time=ScalarParameter() {parameterName=T}
   Speed=ScalarParameter() {defaultValue=0.5, parameterName=Pan Speed}
  ) {speedX=0.1, speedY=0.1}
  Multiply(B=Power(
    Base=TextureSample(UVs=@n27) {texture=/FortWater/Textures/T_WaterFlow_01_Foam_Tiled}.R
    Exp=ScalarParameter() {defaultValue=1, parameterName=power}
   ))
 ]
}
