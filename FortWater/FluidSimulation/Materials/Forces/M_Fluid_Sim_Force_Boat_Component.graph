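# Additive unlit material for a boat's force component in the water fluid simulation, allowing negative values for signed force.
# A procedural wake shape, oriented by X/Y vectors, puts force F in red and foam from |F| above a threshold in green, modulated by panning noise.
# The graph also holds a BoatForceFoam texture variant and a sphere mask offset by height against the landscape water texture; no material output connections appear.
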Material /FortWater/FluidSimulation/Materials/Forces/M_Fluid_Sim_Force_Boat_Component
graph {
 properties {
  bAllowNegativeEmissiveColor=true
  bUseTranslucencyVertexFog=false
  bUsedWithNanite=true
  blendMode=BLEND_Additive
  pixelDepthOffsetMode=PDOM_Legacy
  shadingModel=MSM_Unlit
  usedShadingModels=Unlit
 }
 nodes [
  n1 = Constant() {r=0.5}
  n2 = ScalarParameter() {parameterName=Z}
  n3 = VectorParameter() {parameterName=P, defaultValue=<rgba 0.5 0.5 0.01 0>}
  n4 = Subtract(
   A=ComponentMask(None=ComponentMask(None=@n3.RGB) {g=true, r=true}) {g=true, r=true}
   B=ComponentMask(None=CollectionParameter() {collection=/Landmass/Landscape/BlueprintBrushes/MPC/MPC_Landscape, parameterName=LandscapeLocation}) {g=true, r=true}
  )
  n5 = ComponentMask(None=CollectionParameter() {collection=&MPC_Landscape, parameterName=RTWorldSize}) {g=true, r=true}
  n6 = TextureSample!(
   UVs=Saturate(None=Divide(A=@n4, B=@n5))
   Tex=TextureObjectParameter() {parameterName=WaterVelocityAndHeight, texture=&Black}
   Level=ScalarParameter() {parameterName=WaterVelocityAndHeightMipLevel}
   Apply View MipBias
  ) {mipValueMode=TMVM_MipLevel, texture=&Black}
  n7 = Divide(A=Subtract(A=@n3.B, B=@n6.B), B=Multiply(A=ScalarParameter() {defaultValue=32, parameterName=R}) {constB=2})
  n8 = ScalarParameter() {defaultValue=1, parameterName=LandscapeWaterSystem}
  n9 = SphereMask(
   A=AppendVector(A=TextureCoordinate(), B=Constant())
   B=MaterialFunctionCall(X=@n1, Y=@n1, Z=LinearInterpolate(A=@n2, B=@n7, Alpha=@n8)) {materialFunction=&MakeFloat3}.Result
   Radius=Constant() {r=0.5}
   Hardness=Constant()
  )
  n10 = Multiply(A=@n9, B=ScalarParameter() {defaultValue=1, parameterName=F})
  n11 = OneMinus(None=@n9)
  n12 = Multiply(B=MaterialFunctionCall!(Alpha=Multiply(A=@n11, B=@n11), Offset=ScalarParameter() {defaultValue=0.25, parameterName=fjh}) {materialFunction=/Engine/Functions/Engine_MaterialFunctions01/Opacity/AlphaOffset}.Result)
  Abs(None=AppendVector(A=Multiply(A=@n10), B=Multiply(A=@n12) {constB=0}))
  Add(
   A=CollectionParameter() {collection=&MPC_Landscape, parameterName=LandscapeZLocation}
   B=Multiply(A=CollectionParameter() {collection=&MPC_Landscape, parameterName=LandscapeZScale}, B=@n6.B)
  )
  n13 = VectorParameter() {parameterName=X, defaultValue=<rgba 1 0 0 0>}
  n14 = MaterialFunctionCall(Float2=Subtract(A=TextureCoordinate()) {constB=0.5, desc=offset}) {materialFunction=&BreakOutFloat2Components}
  n15 = VectorParameter() {parameterName=Y, defaultValue=<rgba 0 1 0 0>}
  n16 = Add(A=Multiply(A=ComponentMask(None=@n13.RGB) {g=true, r=true}, B=@n14.G), B=Multiply(A=ComponentMask(None=@n15.RGB) {g=true, r=true}, B=@n14.R))
  n17 = TextureSample!(UVs=Clamp(None=Add(A=@n16) {constB=0.5}), Tex, Apply View MipBias) {texture=/FortWater/Textures/BoatForceFoam}
  n18 = ScalarParameter() {defaultValue=1, parameterName=F}
  n19 = Abs(None=ScalarParameter() {defaultValue=1, parameterName=F})
  n20 = ScalarParameter() {defaultValue=0.25, parameterName=MinFoamForce}
  n21 = Multiply(A=Max(A=Subtract(A=@n19, B=@n20)) {constB=0}, B=ScalarParameter() {defaultValue=0.5, parameterName=Foam Strength}) {constB=10}
  AppendVector(A=AppendVector(A=Multiply(A=@n17.R, B=@n18), B=Multiply(A=@n17.G, B=@n21)), B=Constant())
  Constant3Vector()
  Constant()
  MaterialFunctionCall() {materialFunction=&AlphaOffset}
  MaterialFunctionCall!(Far Mask, Texture, Position) {materialFunction=/FortWater/Materials/Functions/WaterBodyVelocityAndHeight}
  n22 = ComponentMask(None=VectorParameter() {parameterName=X, defaultValue=<rgba 1 0 0 0>}.RGB) {g=true, r=true}
  n23 = MaterialFunctionCall(XY=TextureCoordinate()) {materialFunction=&Swizzle}
  n24 = MaterialFunctionCall(Float2=Multiply(A=Subtract(A=@n23.YX) {constB=0.5, desc=offset}, B=ScalarParameter() {defaultValue=1.25, parameterName=ScaleTexture, sliderMin=0.01}) {constB=2, desc=scale}) {materialFunction=&BreakOutFloat2Components}
  n25 = ComponentMask(None=VectorParameter() {parameterName=Y, defaultValue=<rgba 0 1 0 0>}.RGB) {g=true, r=true}
  n26 = ComponentMask(None=Add(A=Multiply(A=@n22, B=@n24.R), B=Multiply(A=@n25, B=@n24.G))) {g=true, r=true}
  n27 = Abs(None=Multiply(A=ComponentMask(None=@n26) {r=true}) {constB=2})
  n28 = OneMinus(None=@n27)
  n29 = Power(Base=ComponentMask(None=Add(A=@n26) {constB=0.5}) {g=true}) {constExponent=1.5}
  n30 = Subtract(A=OneMinus(None=@n29))
  n31 = Subtract(A=OneMinus(None=Add(A=@n28, B=@n30)))
  n32 = Multiply(A=OneMinus(None=Abs(None=@n31))) {constB=20}
  n33 = LinearInterpolate(Alpha=Multiply(A=@n29, B=@n27)) {constA=20, constB=12}
  n34 = Multiply(A=Clamp(None=Subtract(A=@n32, B=@n33) {constB=19}), B=@n29)
  n35 = ScalarParameter() {defaultValue=1, parameterName=F}
  n36 = Multiply(A=Add(A=OneMinus(None=@n29)) {constB=-0.8}) {constB=1.05}
  n37 = Clamp(None=Multiply(A=Add(A=@n28, B=@n36), B=@n29))
  n38 = Abs(None=ScalarParameter() {defaultValue=1, parameterName=F})
  n39 = ScalarParameter() {defaultValue=0.25, parameterName=MinFoamForce}
  n40 = Multiply(A=Max(A=Subtract(A=@n38, B=@n39)) {constB=0}, B=ScalarParameter() {defaultValue=0.5, parameterName=Foam Strength}) {constB=10}
  n41 = Panner(
   Coordinate=Multiply(A=TextureCoordinate(), B=ScalarParameter() {defaultValue=1, parameterName=NoiseTiling})
   Time=ScalarParameter() {parameterName=T}
   Speed=ScalarParameter() {defaultValue=0.5, parameterName=Pan Speed}
  ) {speedX=0.1, speedY=0.1}
  Multiply(
   A=AppendVector(A=Multiply(A=@n34, B=@n35), B=Multiply(A=@n37, B=@n40)) {desc=R is Force G is Foam}
   B=Power(
    Base=TextureSample(UVs=@n41) {texture=/FortWater/Textures/T_WaterFlow_01_Foam_Tiled}.R
    Exp=ScalarParameter() {defaultValue=0.5, parameterName=power}
   )
  )
  Multiply(A=@n7) {constB=0.85}
  ScalarParameter() {defaultValue=1, parameterName=Final Opacity}
  WorldPosition() {worldPositionShaderOffset=WPT_ExcludeAllShaderOffsets}
 ]
}
