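# Unlit additive material for a round force in the water fluid simulation, packing ripple and foam amounts into color channels.
# Red is a sphere mask scaled by Ripple Intensity, with negative emissive allowed for signed forces.
# Green is a second mask scaled by Foam Strength and the absolute intensity above MinFoamForce, clamped at zero, so weak forces add no foam.
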Material /FortWater/FluidSimulation/Materials/Forces/M_Fluid_Sim_Force_Round
graph {
 properties {
  bAllowNegativeEmissiveColor=true
  bUseTranslucencyVertexFog=false
  bUsedWithNanite=true
  blendMode=BLEND_Additive
  pixelDepthOffsetMode=PDOM_Legacy
  shadingModel=MSM_Unlit
  usedShadingModels=Unlit
 }
 nodes [
  n1 = TextureSampleParameter2D() {parameterName=Ripple, texture=/Game/Effects/Fort_Effects/Textures/Gameplay/T_Sphere_Mask}
  n2 = ScalarParameter() {defaultValue=1, parameterName=Ripple Intensity}
  n3 = TextureSampleParameter2D() {parameterName=Foam Erode, texture=/Game/Effects/Fort_Effects/Textures/Mask/T_Sphere_A}
  n4 = Abs(None=ScalarParameter() {defaultValue=1, parameterName=Ripple Intensity})
  n5 = ScalarParameter() {defaultValue=0.25, parameterName=MinFoamForce}
  n6 = Multiply(A=Max(A=Subtract(A=@n4, B=@n5)) {constB=0}, B=ScalarParameter() {defaultValue=0.5, parameterName=Foam Strength}) {constB=10}
  AppendVector(A=AppendVector(A=Multiply(A=@n1.R, B=@n2), B=Multiply(A=@n3.G, B=@n6)), B=Constant())
  ScalarParameter() {defaultValue=1, parameterName=Final Opacity}
 ]
}
