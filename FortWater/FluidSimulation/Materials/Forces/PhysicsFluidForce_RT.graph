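# Additive unlit material that turns physics-field linear velocity into force for the water fluid simulation, apparently drawn into a render target.
# It maps UVs to world XY using SimLocation and FluidSimSize from MPC_Water, optionally taking Z from the water zone's WaterVelocityAndHeight texture plus a shift.
# The velocity field sampled there becomes a scalar: its length minus a threshold, clamped at zero and scaled by Force.
# Only the deferred shading path produces this value; others output zero.
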Material /FortWater/FluidSimulation/Materials/Forces/PhysicsFluidForce_RT
graph {
 properties {bUsedWithNanite=true, blendMode=BLEND_Additive, pixelDepthOffsetMode=PDOM_Legacy, shadingModel=MSM_Unlit, usedShadingModels=Unlit}
 nodes [
  n1 = Subtract(A=TextureCoordinate()) {constB=0.5}
  n2 = CollectionParameter() {collection=/Water/Materials/MPC/MPC_Water, parameterName=FluidSimSize}
  n3 = CollectionParameter() {collection=&MPC_Water, parameterName=SimLocation}
  n4 = Subtract(A=TextureCoordinate()) {constB=0.5}
  n5 = CollectionParameter() {collection=&MPC_Water, parameterName=FluidSimSize}
  n6 = CollectionParameter() {collection=&MPC_Water, parameterName=SimLocation}
  n7 = MaterialFunctionCall() {materialFunction=/Water/Materials/Functions/WaterZoneData}
  n8 = Divide(A=@n7.Extent)
  n9 = Subtract(
   A=Add(A=Multiply(A=@n4, B=@n5), B=ComponentMask(None=@n6) {g=true, r=true})
   B=ComponentMask(None=Subtract(A=@n7.Location, B=@n8)) {g=true, r=true}
  )
  n10 = ComponentMask(None=@n7.Location) {g=true, r=true}
  n11 = TextureSample!(
   UVs=Saturate(None=Divide(A=@n9, B=@n10))
   Tex=TextureObjectParameter() {parameterName=WaterVelocityAndHeight, texture=&Black}
   Level
   Apply View MipBias
  ) {constMipValue=0, mipValueMode=TMVM_MipLevel, samplerSource=SSM_Wrap_WorldGroupSettings, texture=&Black}
  n12 = ScalarParameter() {defaultValue=-2, parameterName=WaterZShift}
  n13 = AppendVector(
   A=Add(A=Multiply(A=@n1, B=@n2) {constB=2048}, B=ComponentMask(None=@n3) {g=true, r=true})
   B=StaticSwitchParameter(True=Add(A=@n11.B, B=@n12), False=Constant()) {defaultValue=true, parameterName=Enable Water Terrain Texture}
  )
  n14 = Subtract(
   A=MaterialFunctionCall(Vector 3=SamplePhysicsVectorField!(World Position=@n13) {fieldTarget=Vector_LinearVelocity}) {materialFunction=/Engine/Functions/Engine_MaterialFunctions02/Utility/VectorLength}.V3 Length
   B=ScalarParameter() {defaultValue=0.4, parameterName=Threshold}
  )
  ShadingPathSwitch(Default=Constant(), Deferred=Multiply(A=Max(A=@n14) {constB=0}, B=ScalarParameter() {defaultValue=1, parameterName=Force})) {desc=Migrated from FeatureLevelSwitch}
 ]
}
